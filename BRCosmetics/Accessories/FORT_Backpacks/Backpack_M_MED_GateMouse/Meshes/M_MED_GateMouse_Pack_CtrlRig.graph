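# Control Rig for the GateMouse backpack cosmetic, which keeps the pack attached to the moving left shoulder and animates its robot-arm buckle.
# Weighted translation-only parent constraints place the pack root, target and up-vector bones between two dynamic shoulder bones.
# An aim constraint then orients the root toward the target using that up vector.
# An interpolated isJumping value weights a buckle rotation; isTargeting is declared but unused.
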ControlRigBlueprint /BRCosmetics/Accessories/FORT_Backpacks/Backpack_M_MED_GateMouse/Meshes/M_MED_GateMouse_Pack_CtrlRig
graphs [
 CollapseNode_ContainedGraph {nodes [
   n1 = RigVMFunctionEntryNode()
   &RigUnit_ParentConstraint(
    AdvancedSettings=(RotationOrderForFilter=XZY)
    Child=(Type=Bone,Name="pack_root")
    Filter=(TranslationFilter=(bX=True,bY=True,bZ=True),RotationFilter=(bX=False,bY=False,bZ=False),ScaleFilter=(bX=False,bY=False,bZ=False))
    Parents=((Item=(Type=Bone,Name="dyn_shoulder_02_l"),Weight=0.1912),(Item=(Type=Bone,Name="dyn_shoulder_03_l"),Weight=0.8088))
    Weight=1
    ExecutePin=@n1.ExecutePin
   )
   > &RigUnit_ParentConstraint(
    AdvancedSettings=(RotationOrderForFilter=XZY)
    Child=(Type=Bone,Name="pack_target")
    Filter=(TranslationFilter=(bX=True,bY=True,bZ=True),RotationFilter=(bX=False,bY=False,bZ=False),ScaleFilter=(bX=False,bY=False,bZ=False))
    Parents=((Item=(Type=Bone,Name="dyn_shoulder_02_l"),Weight=0.4111),(Item=(Type=Bone,Name="dyn_shoulder_03_l"),Weight=0.5889))
    Weight=1
   )
   > &RigUnit_ParentConstraint(
    AdvancedSettings=(RotationOrderForFilter=XZY)
    Child=(Type=Bone,Name="pack_up_vector")
    Filter=(TranslationFilter=(bX=True,bY=True,bZ=True),RotationFilter=(bX=False,bY=False,bZ=False),ScaleFilter=(bX=False,bY=False,bZ=False))
    Parents=((Item=(Type=Bone,Name="dyn_shoulder_02_l"),Weight=0.1616),(Item=(Type=Bone,Name="dyn_shoulder_03_l"),Weight=0.8384))
    Weight=1
   )
   > &RigUnit_AimConstraintLocalSpaceOffset(
    Child=(Type=Bone,Name="pack_root")
    Filter=(bX=True,bY=True,bZ=True)
    Parents=((Item=(Type=Bone,Name="pack_target"),Weight=1))
    Weight=1
    WorldUp=(Target=(X=0,Y=0,Z=1),Kind=Location,Space=(Type=Bone,Name="pack_up_vector"))
   )
   n2 = &RigVMFunction_AlphaInterp(InterpSpeedDecreasing=5, InterpSpeedIncreasing=5, Scale=1, bClampResult=False, Value=RigVMVariableNode() {variable=isJumping}.Value)
   > n3 = &RigUnit_ModifyTransforms(
    ItemToModify=((Item=(Type=Bone,Name="kin_robo_buckle"),Transform=(Rotation=(X=0,Y=0.5,Z=0,W=0.866),Translation=(X=0,Y=0,Z=0),Scale3D=(X=1,Y=1,Z=1))))
    WeightMinimum=0
    Weight=@n2.Result
   )
   RigVMFunctionReturnNode(ExecutePin=@n3.ExecutePin)
  ]}
 RigVMFunctionLibrary {links [], nodes []}
 RigVMModel {nodes [
   RigUnit_BeginExecution = &RigUnit_BeginExecution()
   Constraints = RigVMCollapseNode(ExecutePin=@RigUnit_BeginExecution.ExecutePin)
  ]}
]
hierarchy [
 root {loc=<0 0 0>, rot=<0 0 0 1>}
 pelvis {parent=root, loc=<0 -0.7321 92.4218>, rot=<0.0127 0.707 -0.0127 -0.707>}
 spine_01 {parent=pelvis, loc=<5.6002 -0.5128 0>, rot=<0 0 0.034 -0.9994>}
 spine_02 {parent=spine_01, loc=<8.0756 0.0139 0>, rot=<0 0 0.0238 -0.9997>}
 spine_03 {parent=spine_02, loc=<7.0395 0 0>, rot=<0 0 0.0149 -0.9999>}
 spine_04 {parent=spine_03, loc=<8.1023 0 0>, rot=<0 0 -0.0132 -0.9999>}
 spine_05 {parent=spine_04, loc=<10.0141 0 0>, rot=<0 0 -0.0015 -1>}
 clavicle_l {parent=spine_05, loc=<2.5856 -0.2866 -4.3777>, rot=<0.0465 -0.7289 -0.009 -0.6829>}
 base_shoulder_01_l {parent=clavicle_l, loc=<13.3661 8.4946 8.4852>, rot=<-0.3928 -0.586 0.5694 -0.4219>}
 dyn_shoulder_01_l {parent=base_shoulder_01_l, loc=<0 0 0>, rot=<0 0 0 1>}
 base_shoulder_02_l {parent=clavicle_l, loc=<15.7063 -2.3042 13.1043>, rot=<-0.3928 -0.586 0.5694 -0.4219>}
 dyn_shoulder_02_l {parent=base_shoulder_02_l, loc=<0 0 0>, rot=<0 0 0 1>}
 base_shoulder_03_l {parent=clavicle_l, loc=<13.2693 -11.2862 5.1789>, rot=<-0.3928 -0.586 0.5694 -0.4219>}
 dyn_shoulder_03_l {parent=base_shoulder_03_l, loc=<0 0 0>, rot=<0 0 0 1>}
 pack_root {parent=clavicle_l, loc=<14.2797 -12.3141 7.5769>, rot=<-0.5368 0.2546 -0.5301 -0.605>}
 kin_roboArm_1 {parent=pack_root, loc=<-17.9931 -2.0052 -0.7512>, rot=<0.079 0.1978 -0.0814 -0.9737>}
 kin_roboArm_2 {parent=kin_roboArm_1, loc=<17.7943 -0.0794 0.2422>, rot=<0.0262 0.5958 -0.0274 -0.8023>}
 kin_roboArm_2_end {parent=kin_roboArm_2, loc=<5.9077 0 0>, rot=<0 0 0 1>}
 kin_robo_buckle {parent=kin_roboArm_2, loc=<11.4417 -0.5332 -2.5329>, rot=<0 0 0 1>}
 pack_target {parent=clavicle_l, loc=<15.0305 -11.1618 9.8262>, rot=<-0.5096 0.2187 -0.5551 -0.62>}
 pack_up_vector {parent=clavicle_l, loc=<12.0608 -12.2788 9.2714>, rot=<-0.5368 0.2546 -0.5301 -0.605>}
]
variables [
 isJumping {type=double, default=0}
 isTargeting {type=bool, default=False}
]
