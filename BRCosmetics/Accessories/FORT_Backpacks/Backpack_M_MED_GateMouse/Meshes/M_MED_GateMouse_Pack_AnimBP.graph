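# Animation Blueprint for the GateMouse backpack cosmetic, posing its shoulder, robot-arm and buckle bones on the character.
# The pose copies from the attached parent, gets connector and left-shoulder bone edits, then runs through Control Rig (targeting, jumping), Anim Dynamics springs and a clavicle_l constraint on pack_root.
# The event graph caches body-mesh shoulder transforms when those bones exist and sets gravity, low-LOD and front-end versus BR simulation-space values.
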AnimBlueprint /BRCosmetics/Accessories/FORT_Backpacks/Backpack_M_MED_GateMouse/Meshes/M_MED_GateMouse_Pack_AnimBP
target_skeleton=&/M_MED_GateMouse_Pack_Skeleton
graphs [
 AnimGraph {nodes [
   n1 = "Constraint - pack_root (P)"!(ComponentPose=Local To Component(LocalPose=Use cached pose 'CR_Cache_Dynamics'()), bAlphaBoolEnabled:bool=True, ConstraintWeights_0=1) -> (Pose) {alphaInputType=Bool, boneToModify {boneName=pack_root}, constraintSetup [{targetBone {boneName=clavicle_l}}], constraintWeights [1]}
   Output Pose(Result=Component To Local(ComponentPose=@n1)) {layerGroup=None}
   n2 = "Transform (Modify) Bone - Bone: dyn_connector_piece"(
    ComponentPose=Local To Component(LocalPose=Copy Pose From Mesh() {bCopyCurves=true, bUseAttachedParent=true})
    Alpha=1
    Translation="0.071,0.4,-0.272"
    Rotation="5.413,0,9.642"
    Scale="1,1,1"
   ) {
    rotationMode=BMM_Additive
    rotationSpace=BCS_BoneSpace
    scaleMode=BMM_Additive
    scaleSpace=BCS_BoneSpace
    translationMode=BMM_Additive
    translationSpace=BCS_BoneSpace
    boneToModify {boneName=dyn_connector_piece}
    rotation {pitch=5.413, roll=9.642}
    translation=<0.071 0.4 -0.272>
   }
   n3 = "Transform (Modify) Bone - Bone: base_shoulder_01_l"!(ComponentPose=@n2, bAlphaBoolEnabled:bool=True, Translation="0,0,0", Rotation="0,0,0") -> (Pose) {
    alphaInputType=Bool
    rotationMode=BMM_Replace
    rotationSpace=BCS_ParentBoneSpace
    translationMode=BMM_Replace
    translationSpace=BCS_ParentBoneSpace
    alphaBoolBlend {blendOption=Linear}
    boneToModify {boneName=base_shoulder_01_l}
   }
   n4 = "Transform (Modify) Bone - Bone: dyn_shoulder_01_l"!(ComponentPose=@n3, bAlphaBoolEnabled:bool=True, Translation="0,0,0", Rotation="0,0,0") -> (Pose) {
    alphaInputType=Bool
    rotationMode=BMM_Replace
    rotationSpace=BCS_ParentBoneSpace
    translationMode=BMM_Replace
    translationSpace=BCS_ParentBoneSpace
    alphaBoolBlend {blendOption=Linear}
    boneToModify {boneName=dyn_shoulder_01_l}
   }
   n5 = "Transform (Modify) Bone - Bone: base_shoulder_02_l"!(ComponentPose=@n4, bAlphaBoolEnabled:bool=True, Translation="0,0,0", Rotation="0,0,0") -> (Pose) {
    alphaInputType=Bool
    rotationMode=BMM_Replace
    rotationSpace=BCS_ParentBoneSpace
    translationMode=BMM_Replace
    translationSpace=BCS_ParentBoneSpace
    alphaBoolBlend {blendOption=Linear}
    boneToModify {boneName=base_shoulder_02_l}
   }
   n6 = "Transform (Modify) Bone - Bone: dyn_shoulder_02_l"!(ComponentPose=@n5, bAlphaBoolEnabled:bool=True, Translation="0,0,0", Rotation="0,0,0") -> (Pose) {
    alphaInputType=Bool
    rotationMode=BMM_Replace
    rotationSpace=BCS_ParentBoneSpace
    translationMode=BMM_Replace
    translationSpace=BCS_ParentBoneSpace
    alphaBoolBlend {blendOption=Linear}
    boneToModify {boneName=dyn_shoulder_02_l}
   }
   n7 = "Transform (Modify) Bone - Bone: base_shoulder_03_l"!(ComponentPose=@n6, bAlphaBoolEnabled:bool=True, Translation="0,0,0", Rotation="0,0,0") -> (Pose) {
    alphaInputType=Bool
    rotationMode=BMM_Replace
    rotationSpace=BCS_ParentBoneSpace
    translationMode=BMM_Replace
    translationSpace=BCS_ParentBoneSpace
    alphaBoolBlend {blendOption=Linear}
    boneToModify {boneName=base_shoulder_03_l}
   }
   n8 = "Transform (Modify) Bone - Bone: dyn_shoulder_03_l"!(ComponentPose=@n7, bAlphaBoolEnabled:bool=True, Translation="0,0,0", Rotation="0,0,0") -> (Pose) {
    alphaInputType=Bool
    rotationMode=BMM_Replace
    rotationSpace=BCS_ParentBoneSpace
    translationMode=BMM_Replace
    translationSpace=BCS_ParentBoneSpace
    alphaBoolBlend {blendOption=Linear}
    boneToModify {boneName=dyn_shoulder_03_l}
   }
   n9 = Control Rig!(
    Source=Component To Local(ComponentPose=@n8)
    Alpha
    isTargeting:bool=Get bIsTargeting()
    isJumping:double=Get IsJumping() -> (IsJumping:double)
   ) -> (Pose) {controlRigAssetReference {blueprintRigClass=&/M_MED_GateMouse_Pack_CtrlRig.M_MED_GateMouse_Pack_CtrlRig_C}}
   Save cached pose 'CR_Cache_Input'(Pose=@n9) {cacheName=CR_Cache_Input}
   n10 = Anim Dynamics - Bone: kin_roboArm_1!(
    ComponentPose=Local To Component(LocalPose=Use cached pose 'CR_Cache_Input'())
    Alpha=1
    GravityOverride:Vector="0,0,-980"
    SimSpaceSettings:Anim Phys Sim Space Settings Structure=(SimSpaceAngularAlpha=0,MaxAngularVelocity=10000,MaxAngularAcceleration=10000,ExternalAngularVelocity=(X=0,Y=0,Z=0))
    ExternalForce="0,0,0"
   ) -> (Pose) {
    angularDampingOverride=1
    angularSpringConstant=20000
    bAngularSpring=true
    bOverrideAngularDamping=true
    bUseGravityOverride=true
    gravityScale=0.35
    boundBone {boneName=kin_roboArm_1}
    componentAppliedLinearAccClamp=<10000 10000 10000>
    componentLinearAccScale=<1 1 1>
    componentLinearVelScale=<1 1 1>
    gravityOverride {z=-980}
    physicsBodyDefinitions [{
      boundBone {boneName=kin_roboArm_1}
      boxExtents=<40 50 50>
      constraintSetup {angularLimitsMax=<0 0 0>, angularLimitsMin=<0 -10 0>, angularTarget=<1 0 0>}
      localJointOffset=<2 0 0>
     }]
   }
   n11 = Anim Dynamics - Bone: kin_roboArm_2!(
    ComponentPose=@n10
    Alpha=0
    GravityOverride:Vector="0,0,-980"
    SimSpaceSettings:Anim Phys Sim Space Settings Structure=(SimSpaceAngularAlpha=0,MaxAngularVelocity=10000,MaxAngularAcceleration=10000,ExternalAngularVelocity=(X=0,Y=0,Z=0))
    ExternalForce="0,0,0"
   ) -> (Pose) {
    angularDampingOverride=1
    angularSpringConstant=20000
    bAngularSpring=true
    bOverrideAngularDamping=true
    bUseGravityOverride=true
    gravityScale=0.35
    simulationSpace=RootRelative
    boundBone {boneName=kin_roboArm_2}
    componentAppliedLinearAccClamp=<10000 10000 10000>
    componentLinearAccScale=<1 1 1>
    componentLinearVelScale=<1 1 1>
    gravityOverride {z=-980}
    physicsBodyDefinitions [{
      boundBone {boneName=kin_roboArm_2}
      boxExtents=<40 50 50>
      constraintSetup {angularLimitsMax=<0 0 5>, angularLimitsMin=<0 0 -5>, angularTarget=<1 0 0>}
      localJointOffset=<2 0 0>
     }]
   }
   n12 = Anim Dynamics - Bone: kin_robo_buckle!(
    ComponentPose=@n11
    Alpha=1
    GravityOverride:Vector="0,0,-980"
    SimSpaceSettings:Anim Phys Sim Space Settings Structure=(SimSpaceAngularAlpha=0,MaxAngularVelocity=10000,MaxAngularAcceleration=10000,ExternalAngularVelocity=(X=0,Y=0,Z=0))
    ExternalForce="0,0,0"
   ) -> (Pose) {
    angularDampingOverride=1
    angularSpringConstant=20000
    bAngularSpring=true
    bOverrideAngularDamping=true
    bUseGravityOverride=true
    gravityScale=0.35
    boundBone {boneName=kin_robo_buckle}
    componentAppliedLinearAccClamp=<10000 10000 10000>
    componentLinearAccScale=<1 1 1>
    componentLinearVelScale=<1 1 1>
    gravityOverride {z=-980}
    physicsBodyDefinitions [{
      boundBone {boneName=kin_robo_buckle}
      boxExtents=<30 40 40>
      constraintSetup {angularLimitsMax=<0 70 0>, angularLimitsMin=<0 0 0>, angularTarget=<1 0 0>}
      localJointOffset=<4 0 0>
     }]
   }
   Save cached pose 'CR_Cache_Dynamics'(Pose=Component To Local(ComponentPose=@n12)) {cacheName=CR_Cache_Dynamics}
  ]}
 EventGraph {nodes [
   Event BlueprintUpdateAnimation()
   > n1 = Sequence!(execute) -> (then_0, then_1, then_2:Exec)
   n2 = Get BodyMesh() -> (BodyMesh:Skeletal Mesh Component Budgeted Object Reference)
   Is Valid(exec=@n1.then_1, InputObject=@n2)
   n3 = GetSocketBoneName!(self:Skinned Mesh Component Object Reference=@n2, InSocketName:Name=dyn_shoulder_01_l) -> (ReturnValue:Name=None)
   > n4 = Set isUsingItsBody(isUsingItsBody:bool="Not Equal Exactly (String)"(A="To String (Name)"(InName=@n3), B=None)) -> (Output_Get:bool)
   > Branch(Condition=@n4)
   n5 = Get BodyMesh()
   n6 = BreakTransform(InTransform=GetSocketTransform(self=@n5, InSocketName=base_shoulder_01_l, TransformSpace=RTS_ParentBoneSpace))
   > Set base_shoulder_01_l_T(base_shoulder_01_l_T:Vector=@n6) -> (Output_Get:Vector)
   > Set base_shoulder_01_l_R(base_shoulder_01_l_R:Rotator=@n6.Rotation) -> (Output_Get:Rotator)
   n7 = BreakTransform(InTransform=GetSocketTransform(self=@n5, InSocketName=dyn_shoulder_01_l, TransformSpace=RTS_ParentBoneSpace))
   > Set dyn_shoulder_01_l_T(dyn_shoulder_01_l_T:Vector=@n7) -> (Output_Get:Vector)
   > Set dyn_shoulder_01_l_R(dyn_shoulder_01_l_R:Rotator=@n7.Rotation) -> (Output_Get:Rotator)
   n8 = BreakTransform(InTransform=GetSocketTransform(self=@n5, InSocketName=base_shoulder_02_l, TransformSpace=RTS_ParentBoneSpace))
   > Set base_shoulder_02_l_T(base_shoulder_02_l_T:Vector=@n8) -> (Output_Get:Vector)
   > Set base_shoulder_02_l_R(base_shoulder_02_l_R:Rotator=@n8.Rotation) -> (Output_Get:Rotator)
   n9 = BreakTransform(InTransform=GetSocketTransform(self=@n5, InSocketName=dyn_shoulder_02_l, TransformSpace=RTS_ParentBoneSpace))
   > Set dyn_shoulder_02_l_T(dyn_shoulder_02_l_T:Vector=@n9) -> (Output_Get:Vector)
   > Set dyn_shoulder_02_l_R(dyn_shoulder_02_l_R:Rotator=@n9.Rotation) -> (Output_Get:Rotator)
   n10 = BreakTransform(InTransform=GetSocketTransform(self=@n5, InSocketName=base_shoulder_03_l, TransformSpace=RTS_ParentBoneSpace))
   > Set base_shoulder_03_l_T(base_shoulder_03_l_T:Vector=@n10) -> (Output_Get:Vector)
   > Set base_shoulder_03_l_R(base_shoulder_03_l_R:Rotator=@n10.Rotation) -> (Output_Get:Rotator)
   n11 = BreakTransform(InTransform=GetSocketTransform(self=@n5, InSocketName=dyn_shoulder_03_l, TransformSpace=RTS_ParentBoneSpace))
   > Set dyn_shoulder_03_l_T(dyn_shoulder_03_l_T:Vector=@n11) -> (Output_Get:Vector)
   > Set dyn_shoulder_03_l_R(dyn_shoulder_03_l_R:Rotator=@n11.Rotation) -> (Output_Get:Rotator)
   Branch(execute=@n1.then_2, Condition="integer > integer"(A=Get CurrentLOD(), B=1))
   > Set LowLOD(LowLOD:bool=true) -> (Output_Get:bool)
   n12 = "Gravity Override State Function (Parameter Asset) (Nativized)"(
    execute=@n1.then_0
    CharacterPartAnimInstance=Self-Reference()
    GameTimeInSeconds=GetGameTimeInSeconds()
    Parameters=&/M_MED_GateMouse_Pack_Parameters
   )
   > Set Gravity Override(Gravity Override=@n12)
   > Set Sim Space Settings(Sim Space Settings:Anim Phys Sim Space Settings Structure=Get Sim Space Settings_BR() -> (Sim Space Settings_BR:Anim Phys Sim Space Settings Structure)) -> (Output_Get:Anim Phys Sim Space Settings Structure)
   > Branch(Condition=Get bIsInFrontEnd())
   > Set Sim Space Settings(Sim Space Settings=Get Sim Space Settings_FE() -> (Sim Space Settings_FE:Anim Phys Sim Space Settings Structure))
   n13 = Event OnCharacterCustomizationLoaded()
   > Set BodyMesh(BodyMesh:Skeletal Mesh Component Budgeted Object Reference=Get Array Item with EFort Custom Part Type(Array=@n13.CustomMeshParts)) -> (Output_Get:Skeletal Mesh Component Budgeted Object Reference)
   Event OnBeginJumping()
   > Set IsJumping(IsJumping:double=1) -> (Output_Get:double)
   Event OnEndFalling()
   > Set IsJumping(IsJumping=0)
  ]}
]
variables [
 base_shoulder_01_l_R {type=Rotator}
 base_shoulder_01_l_T {type=Vector}
 base_shoulder_02_l_R {type=Rotator}
 base_shoulder_02_l_T {type=Vector}
 base_shoulder_03_l_R {type=Rotator}
 base_shoulder_03_l_T {type=Vector}
 BodyMesh {type=SkeletalMeshComponentBudgeted}
 dyn_shoulder_01_l_R {type=Rotator}
 dyn_shoulder_01_l_T {type=Vector}
 dyn_shoulder_02_l_R {type=Rotator}
 dyn_shoulder_02_l_T {type=Vector}
 dyn_shoulder_03_l_R {type=Rotator}
 dyn_shoulder_03_l_T {type=Vector}
 Gravity Override {type=Vector}
 IsJumping {type=number}
 isUsingItsBody {type=boolean}
 LowLOD {type=boolean}
 Sim Space Settings {
  type=AnimPhysSimSpaceSettings
  default {maxAngularAcceleration=10000, maxAngularVelocity=10000, simSpaceAngularAlpha=0, externalAngularVelocity=<0 0 0>}
 }
 Sim Space Settings_BR {
  type=AnimPhysSimSpaceSettings
  default {maxAngularAcceleration=10000, maxAngularVelocity=10000, simSpaceAngularAlpha=0, externalAngularVelocity=<0 0 0>}
 }
 Sim Space Settings_FE {
  type=AnimPhysSimSpaceSettings
  default {maxAngularAcceleration=10000, maxAngularVelocity=10000, simSpaceAngularAlpha=0.2, externalAngularVelocity=<0.2 0.2 0.2>}
 }
]
